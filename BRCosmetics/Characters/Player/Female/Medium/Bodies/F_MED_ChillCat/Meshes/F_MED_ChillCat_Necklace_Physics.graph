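# Physics asset for the ChillCat female medium cosmetic body, defining simulation bodies for its necklace, hood and both sleeves.
# The necklace is a two-link box chain from a small sphere base; the hood is one capsule on its base.
# Each sleeve has three boxes constrained to a base sphere, plus an upper-arm twist capsule.
# Limits are tight; only the left sleeve base is constrained to its arm, and its swing limits differ from the right.
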PhysicsAsset /BRCosmetics/Characters/Player/Female/Medium/Bodies/F_MED_ChillCat/Meshes/F_MED_ChillCat_Necklace_Physics
bodies {
 dyn_hood_base [{shapeType=Sphere, radius=0.5, shapeName=dyn_hood_base_sphere}]
 dyn_hood_parent [{shapeType=Capsule, length=6.4133, radius=3.2553, shapeName=dyn_hood_parent_capsule, center=<5.4483 0 0>, rotation=<rot 90 0 0>}]
 dyn_necklace_01 [{shapeType=Box, extentX=8.1993, extentY=10.1331, extentZ=1.5463, shapeName=dyn_necklace_01_box, center=<2.7083 0 0>}]
 dyn_necklace_base [{shapeType=Sphere, radius=0.5, shapeName=dyn_necklace_base_sphere}]
 dyn_necklace_parent [{shapeType=Box, extentX=8.3513, extentY=10.3883, extentZ=2.7632, shapeName=dyn_necklace_parent_box, center=<3.8754 0 0>}]
 dyn_sleeve_01_l [{
   shapeType=Box
   extentX=12.7342
   extentY=14.4026
   extentZ=2.2038
   shapeName=dyn_sleeve_01_r_box
   center=<5.2375 0 4.1715>
   rotation=<rot 0 0 -180>
  }]
 dyn_sleeve_01_r [{shapeType=Box, extentX=12.7342, extentY=14.4026, extentZ=2.2038, shapeName=dyn_sleeve_01_r_box, center=<5.3283 0 -4.1715>}]
 dyn_sleeve_02_l [{
   shapeType=Box
   extentX=12.9533
   extentY=3.5815
   extentZ=14.5122
   shapeName=dyn_sleeve_02_r_box
   center=<5.3503 0 -5.6617>
   rotation=<rot 0 0 -90>
  }]
 dyn_sleeve_02_r [{
   shapeType=Box
   extentX=12.9533
   extentY=3.5815
   extentZ=14.5122
   shapeName=dyn_sleeve_02_r_box
   center=<5.2156 0 5.6617>
   rotation=<rot 0 0 90>
  }]
 dyn_sleeve_03_l [{
   shapeType=Box
   extentX=12.7695
   extentY=12.1967
   extentZ=2.1255
   shapeName=dyn_sleeve_03_r_box
   center=<5.2375 -8.1213 0>
   rotation=<rot 0 0 -90>
  }]
 dyn_sleeve_03_r [{
   shapeType=Box
   extentX=12.7695
   extentY=12.1967
   extentZ=2.1255
   shapeName=dyn_sleeve_03_r_box
   center=<5.3283 8.1213 0>
   rotation=<rot 0 0 90>
  }]
 dyn_sleeve_base_l [{shapeType=Sphere, radius=0.5, shapeName=dyn_sleeve_base_r_sphere, center=<10.5658 0 0>}]
 dyn_sleeve_base_r [{shapeType=Sphere, radius=0.5, shapeName=dyn_sleeve_base_r_sphere}]
 upperarm_twist_01_l [{shapeType=Capsule, length=25, radius=5, shapeName=upperarm_twist_01_r_capsule, rotation=<rot 90 180 0>}]
 upperarm_twist_01_r [{shapeType=Capsule, length=25, radius=5, shapeName=upperarm_twist_01_r_capsule, rotation=<rot 90 0 0>}]
}
constraints [
 {bone1Name=dyn_necklace_parent, bone2Name=dyn_necklace_base, swing1LimitDegrees=25, swing2LimitDegrees=5, twistLimitDegrees=1}
 {bone1Name=dyn_necklace_01, bone2Name=dyn_necklace_parent, swing1LimitDegrees=20, swing2LimitDegrees=5, twistLimitDegrees=1}
 {bone1Name=dyn_sleeve_01_r, bone2Name=dyn_sleeve_base_r, swing1LimitDegrees=10, swing2LimitDegrees=5, twistLimitDegrees=1}
 {bone1Name=dyn_sleeve_02_r, bone2Name=dyn_sleeve_base_r, swing1LimitDegrees=10, swing2LimitDegrees=5, twistLimitDegrees=1}
 {bone1Name=dyn_sleeve_03_r, bone2Name=dyn_sleeve_base_r, swing1LimitDegrees=10, swing2LimitDegrees=5, twistLimitDegrees=1}
 {bone1Name=dyn_sleeve_base_l, bone2Name=upperarm_twist_01_l, swing1LimitDegrees=10, swing2LimitDegrees=10, twistLimitDegrees=1}
 {bone1Name=dyn_sleeve_01_l, bone2Name=dyn_sleeve_base_l, swing1LimitDegrees=10, swing2LimitDegrees=10, twistLimitDegrees=1}
 {bone1Name=dyn_sleeve_02_l, bone2Name=dyn_sleeve_base_l, swing1LimitDegrees=10, swing2LimitDegrees=10, twistLimitDegrees=1}
 {bone1Name=dyn_sleeve_03_l, bone2Name=dyn_sleeve_base_l, swing1LimitDegrees=10, swing2LimitDegrees=10, twistLimitDegrees=1}
 {bone1Name=dyn_hood_parent, bone2Name=dyn_hood_base, swing1LimitDegrees=30, swing2LimitDegrees=30, twistLimitDegrees=1}
]
