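# Physics asset for the Chill Cat medium female player body cosmetic, covering a front and back dynamic chain plus the right leg.
# Small spheres at the base bones anchor two runs of capsules, linked by swing- and twist-limited constraints, with tighter swing near the back root.
# An angularly free constraint joins the two chain tips, closing them into one hanging loop.
# Right thigh and calf capsules carry no constraints and give the leg collision volume.
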PhysicsAsset /BRCosmetics/Characters/Player/Female/Medium/Bodies/F_MED_ChillCat/Meshes/F_MED_ChillCat_Physics
bodies {
 calf_r [{
   shapeType=Capsule
   length=32.0595
   radius=5.8725
   shapeName=calf_r_capsule
   center=<10.3471 1.9927 0>
   rotation=<rot 84.3182 -173.7594 179.9999>
  }]
 dyn_chain_01_bk [{shapeType=Capsule, length=8.1517, radius=2.5829, shapeName=dyn_chain_01_bk_capsule, center=<4.6286 0 -0.8794>, rotation=<rot 90 0 0>}]
 dyn_chain_01_frnt [{
   shapeType=Capsule
   length=7.4477
   radius=3.5141
   shapeName=dyn_chain_01_frnt_capsule
   center=<3.9438 0 -1.1051>
   rotation=<rot 84.3751 179.9997 179.9997>
  }]
 dyn_chain_02_bk [{shapeType=Capsule, length=7.0578, radius=2.6546, shapeName=dyn_chain_02_bk_capsule, center=<4.6578 0 -1.0807>, rotation=<rot 90 0 0>}]
 dyn_chain_02_frnt [{shapeType=Capsule, length=8.0203, radius=2.4846, shapeName=dyn_chain_02_frnt_capsule, center=<3.5539 0 -0.7983>, rotation=<rot 90 0 0>}]
 dyn_chain_03_bk [{shapeType=Capsule, length=6.211, radius=2.6333, shapeName=dyn_chain_03_bk_capsule, center=<3.4768 0 -1.2734>, rotation=<rot 90 0 0>}]
 dyn_chain_03_frnt [{shapeType=Capsule, length=3.9707, radius=2.0584, shapeName=dyn_chain_03_frnt_capsule, center=<2.8233 0 0>, rotation=<rot 90 0 0>}]
 dyn_chain_04_bk [{
   shapeType=Capsule
   length=7.3373
   radius=2.2151
   shapeName=dyn_chain_04_bk_capsule
   center=<3.2312 0 -0.8281>
   rotation=<rot 78.75 179.9999 179.9998>
  }]
 dyn_chain_05_bk [{shapeType=Capsule, length=5.848, radius=2.332, shapeName=dyn_chain_05_bk_capsule, center=<-1.6188 0 0>, rotation=<rot 90 0 0>}]
 dyn_chain_base_bk [{shapeType=Sphere, radius=0.5, shapeName=dyn_chain_base_bk_sphere}]
 dyn_chain_base_frnt [{shapeType=Sphere, radius=0.5, shapeName=dyn_chain_base_frnt_sphere}]
 dyn_chain_parent_bk [{shapeType=Capsule, length=7.4419, radius=2.5489, shapeName=dyn_chain_parent_bk_capsule, center=<4.4687 0 0>, rotation=<rot 90 0 0>}]
 dyn_chain_parent_frnt [{shapeType=Capsule, length=7.9253, radius=3.6976, shapeName=dyn_chain_parent_frnt_capsule, center=<5.8045 0 0>, rotation=<rot 90 0 0>}]
 thigh_r [
  {
   shapeType=Capsule
   length=32.0946
   radius=5.7931
   shapeName=thigh_r_capsule
   center=<23.1449 0 -0.6624>
   rotation=<rot 88.5573 179.9989 179.9992>
  }
  {shapeType=Capsule, length=6.5245, radius=7.7345, shapeName=thigh_r_capsule, center=<6.2582 -0.2836 -4.53>, rotation=<rot 0 -0.0001 -5.9933>}
 ]
}
constraints [
 {bone1Name=dyn_chain_03_frnt, bone2Name=dyn_chain_02_frnt, swing1LimitDegrees=30, swing2LimitDegrees=30, twistLimitDegrees=3}
 {bone1Name=dyn_chain_parent_frnt, bone2Name=dyn_chain_base_frnt, swing1LimitDegrees=30, swing2LimitDegrees=30, twistLimitDegrees=3}
 {bone1Name=dyn_chain_01_frnt, bone2Name=dyn_chain_parent_frnt, swing1LimitDegrees=30, swing2LimitDegrees=30, twistLimitDegrees=3}
 {bone1Name=dyn_chain_02_frnt, bone2Name=dyn_chain_01_frnt, swing1LimitDegrees=30, swing2LimitDegrees=30, twistLimitDegrees=3}
 {bone1Name=dyn_chain_05_bk, bone2Name=dyn_chain_04_bk, swing1LimitDegrees=30, swing2LimitDegrees=30, twistLimitDegrees=3}
 {bone1Name=dyn_chain_parent_bk, bone2Name=dyn_chain_base_bk, swing1LimitDegrees=10, swing2LimitDegrees=10, twistLimitDegrees=3}
 {bone1Name=dyn_chain_01_bk, bone2Name=dyn_chain_parent_bk, swing1LimitDegrees=20, swing2LimitDegrees=20, twistLimitDegrees=3}
 {bone1Name=dyn_chain_02_bk, bone2Name=dyn_chain_01_bk, swing1LimitDegrees=30, swing2LimitDegrees=30, twistLimitDegrees=3}
 {bone1Name=dyn_chain_03_bk, bone2Name=dyn_chain_02_bk, swing1LimitDegrees=30, swing2LimitDegrees=30, twistLimitDegrees=3}
 {bone1Name=dyn_chain_04_bk, bone2Name=dyn_chain_03_bk, swing1LimitDegrees=30, swing2LimitDegrees=30, twistLimitDegrees=3}
 {
  bone1Name=dyn_chain_03_frnt
  bone2Name=dyn_chain_05_bk
  swing1LimitDegrees=45
  swing1Motion=Free
  swing2LimitDegrees=45
  swing2Motion=Free
  twistLimitDegrees=45
  twistMotion=Free
 }
]
