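# Post-process style Anim Blueprint for the ChillCat female body, copying the attached parent's pose, then adding Control Rig and secondary motion.
# Weighted constraints place tail and can bones, then two RigidBody nodes simulate the necklace and main physics assets.
# Update event limits the tail constraint to LOD 0–1, offsets the tail when skydiving and adjusts can bones in front end.
# Sleeve AnimDynamics chain is unconnected to the output.
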AnimBlueprint /BRCosmetics/Characters/Player/Female/Medium/Bodies/F_MED_ChillCat/Meshes/F_MED_ChillCat_AnimBP
target_skeleton=&/F_MED_ChillCat_Skeleton
graphs [
 AnimGraph {nodes [
   n1 = Anim Dynamics - Bone: dyn_sleeve_01_r(Alpha=1, ExternalForce="0,0,0") {
    angularDampingOverride=0.99
    angularSpringConstant=5000
    bGravityOverrideInSimSpace=true
    bOverrideAngularDamping=true
    alphaBoolBlend {blendOption=Linear}
    alphaScaleBiasClamp {scale=0.2}
    boundBone {boneName=dyn_sleeve_01_r}
    componentAppliedLinearAccClamp=<6000 6000 6000>
    componentLinearAccScale=<0.6 0.6 0.6>
    componentLinearVelScale=<0.6 0.6 0.6>
    physicsBodyDefinitions [{
      sphereCollisionRadius=0
      boundBone {boneName=dyn_sleeve_01_r}
      boxExtents=<10 20 20>
      constraintSetup {angularLimitsMax=<45 20 0>, angularLimitsMin=<-45 -20 -75>, angularTarget=<1 0 0>}
      localJointOffset=<8 0 0>
     }]
    relativeSpaceBone {boneName=spine_05}
   }
   n2 = Anim Dynamics - Bone: dyn_sleeve_02_r(ComponentPose=@n1, Alpha=1, ExternalForce="0,0,0") {
    angularDampingOverride=0.99
    angularSpringConstant=5000
    bGravityOverrideInSimSpace=true
    bOverrideAngularDamping=true
    alphaBoolBlend {blendOption=Linear}
    alphaScaleBiasClamp {scale=0.2}
    boundBone {boneName=dyn_sleeve_02_r}
    componentAppliedLinearAccClamp=<6000 6000 6000>
    componentLinearAccScale=<0.6 0.6 0.6>
    componentLinearVelScale=<0.6 0.6 0.6>
    physicsBodyDefinitions [{
      sphereCollisionRadius=0
      boundBone {boneName=dyn_sleeve_02_r}
      boxExtents=<10 40 40>
      constraintSetup {angularLimitsMax=<45 20 0>, angularLimitsMin=<-45 -20 -75>, angularTarget=<1 0 0>}
      localJointOffset=<8 0 0>
     }]
    relativeSpaceBone {boneName=spine_05}
   }
   n3 = Anim Dynamics - Bone: dyn_sleeve_03_r(ComponentPose=@n2, Alpha=1, ExternalForce="0,0,0") {
    angularDampingOverride=0.99
    angularSpringConstant=5000
    bGravityOverrideInSimSpace=true
    bOverrideAngularDamping=true
    alphaBoolBlend {blendOption=Linear}
    alphaScaleBiasClamp {scale=0.2}
    boundBone {boneName=dyn_sleeve_03_r}
    componentAppliedLinearAccClamp=<6000 6000 6000>
    componentLinearAccScale=<0.6 0.6 0.6>
    componentLinearVelScale=<0.6 0.6 0.6>
    physicsBodyDefinitions [{
      sphereCollisionRadius=0
      boundBone {boneName=dyn_sleeve_03_r}
      boxExtents=<10 30 30>
      constraintSetup {angularLimitsMax=<45 20 0>, angularLimitsMin=<-45 -20 -75>, angularTarget=<1 0 0>}
      localJointOffset=<8 0 0>
     }]
    relativeSpaceBone {boneName=spine_05}
   }
   n4 = Anim Dynamics - Bone: dyn_sleeve_01_l(ComponentPose=@n3, Alpha=1, ExternalForce="0,0,0") {
    angularDampingOverride=0.99
    angularSpringConstant=5000
    bGravityOverrideInSimSpace=true
    bOverrideAngularDamping=true
    alphaBoolBlend {blendOption=Linear}
    alphaScaleBiasClamp {scale=0.2}
    boundBone {boneName=dyn_sleeve_01_l}
    componentAppliedLinearAccClamp=<6000 6000 6000>
    componentLinearAccScale=<0.5 0.5 0.5>
    componentLinearVelScale=<0.5 0.5 0.5>
    physicsBodyDefinitions [{
      sphereCollisionRadius=0
      boundBone {boneName=dyn_sleeve_01_l}
      boxExtents=<10 20 20>
      constraintSetup {angularLimitsMax=<45 20 0>, angularLimitsMin=<-45 -20 -75>, angularTarget=<1 0 0>}
      localJointOffset=<7 0 0>
     }]
   }
   n5 = Anim Dynamics - Bone: dyn_sleeve_02_l(ComponentPose=@n4, Alpha=1, ExternalForce="0,0,0") {
    angularDampingOverride=0.99
    angularSpringConstant=5000
    bGravityOverrideInSimSpace=true
    bOverrideAngularDamping=true
    alphaBoolBlend {blendOption=Linear}
    alphaScaleBiasClamp {scale=0.2}
    boundBone {boneName=dyn_sleeve_02_l}
    componentAppliedLinearAccClamp=<6000 6000 6000>
    componentLinearAccScale=<0.5 0.5 0.5>
    componentLinearVelScale=<0.5 0.5 0.5>
    physicsBodyDefinitions [{
      sphereCollisionRadius=0
      boundBone {boneName=dyn_sleeve_02_l}
      boxExtents=<10 40 40>
      constraintSetup {angularLimitsMax=<45 20 0>, angularLimitsMin=<-45 -20 -75>, angularTarget=<1 0 0>}
      localJointOffset=<7 0 0>
     }]
   }
   Anim Dynamics - Bone: dyn_sleeve_03_l(ComponentPose=@n5, Alpha=1, ExternalForce="0,0,0") {
    angularDampingOverride=0.99
    angularSpringConstant=5000
    bGravityOverrideInSimSpace=true
    bOverrideAngularDamping=true
    alphaBoolBlend {blendOption=Linear}
    alphaScaleBiasClamp {scale=0.2}
    boundBone {boneName=dyn_sleeve_03_l}
    componentAppliedLinearAccClamp=<6000 6000 6000>
    componentLinearAccScale=<0.5 0.5 0.5>
    componentLinearVelScale=<0.5 0.5 0.5>
    physicsBodyDefinitions [{
      sphereCollisionRadius=0
      boundBone {boneName=dyn_sleeve_03_l}
      boxExtents=<10 30 30>
      constraintSetup {angularLimitsMax=<45 20 0>, angularLimitsMin=<-45 -20 -75>, angularTarget=<1 0 0>}
      localJointOffset=<7 0 0>
     }]
   }
   n6 = Control Rig(Source=Copy Pose From Mesh() {bUseAttachedParent=true}) {
    alphaBoolBlend {blendOption=Linear}
    alphaScaleBiasClamp {interpSpeedDecreasing=6, interpSpeedIncreasing=50}
    controlRigAssetReference {blueprintRigClass=&/F_MED_ChillCat_CtrlRig.F_MED_ChillCat_CtrlRig_C}
   }
   n7 = "Constraint - const_tail (PPPPT)"!(
    ComponentPose=Local To Component(LocalPose=@n6)
    bAlphaBoolEnabled:bool=Get LOD_Check() -> (LOD_Check:bool)
    ConstraintWeights_0=0.8806
    ConstraintWeights_1:float=0.06
    ConstraintWeights_2:float=0.048
    ConstraintWeights_3:float=0.004
    ConstraintWeights_4:float=0.0074
   ) -> (Pose) {
    alphaInputType=Bool
    alphaBoolBlend {blendOption=Linear}
    boneToModify {boneName=const_tail}
    constraintSetup [
     {targetBone {boneName=pelvis}}
     {targetBone {boneName=thigh_twist_01_l}}
     {targetBone {boneName=thigh_twist_01_r}}
     {targetBone {boneName=thigh_r}}
     {transformType=Translation, targetBone {boneName=thigh_l}}
    ]
    constraintWeights [1,1,1,1,1]
   }
   n8 = "Transform (Modify) Bone - Bone: const_tail"(ComponentPose=@n7, Alpha=Get Skydive() -> (Skydive:double), Translation="1.5,0,0", Rotation="0,0,0", Scale="1,1,1") {
    translationMode=BMM_Additive
    translationSpace=BCS_BoneSpace
    alphaBoolBlend {blendOption=Linear}
    boneToModify {boneName=const_tail}
    translation {x=1.5}
   }
   n9 = "Constraint - dyn_can_back_base (PPPP)"!(
    ComponentPose=@n8
    Alpha=1
    ConstraintWeights_0=0.3595
    ConstraintWeights_1:float=0.0053
    ConstraintWeights_2:float=0.0569
    ConstraintWeights_3:float=0.5783
   ) -> (Pose) {
    alphaBoolBlend {blendOption=Linear}
    boneToModify {boneName=dyn_can_back_base}
    constraintSetup [
     {targetBone {boneName=upperarm_twist_01_l}}
     {targetBone {boneName=spine_04}}
     {targetBone {boneName=spine_05}}
     {targetBone {boneName=clavicle_l}}
    ]
    constraintWeights [1,1,1,1]
   }
   n10 = "Constraint - dyn_can_mid_base (PPPP)"!(
    ComponentPose=@n9
    Alpha=1
    ConstraintWeights_0=0.2704
    ConstraintWeights_1:float=0.1194
    ConstraintWeights_2:float=0.1615
    ConstraintWeights_3:float=0.4486
   ) -> (Pose) {
    alphaBoolBlend {blendOption=Linear}
    boneToModify {boneName=dyn_can_mid_base}
    constraintSetup [
     {targetBone {boneName=upperarm_twist_01_l}}
     {targetBone {boneName=spine_04}}
     {targetBone {boneName=spine_05}}
     {targetBone {boneName=clavicle_l}}
    ]
    constraintWeights [1,1,1,1]
   }
   n11 = "Constraint - dyn_can_front_base (PPPP)"!(
    ComponentPose=@n10
    Alpha=1
    ConstraintWeights_0=0.0199
    ConstraintWeights_1:float=0.7075
    ConstraintWeights_2:float=0.2098
    ConstraintWeights_3:float=0.0628
   ) -> (Pose) {
    alphaBoolBlend {blendOption=Linear}
    boneToModify {boneName=dyn_can_front_base}
    constraintSetup [
     {targetBone {boneName=upperarm_twist_01_l}}
     {targetBone {boneName=spine_04}}
     {targetBone {boneName=spine_05}}
     {targetBone {boneName=clavicle_l}}
    ]
    constraintWeights [1,1,1,1]
   }
   n12 = Get FrontEndCheck() -> (FrontEndCheck:double)
   n13 = "Transform (Modify) Bone - Bone: dyn_can_mid_base"(ComponentPose=@n11, Alpha=@n12, Translation="-0.000,-0.000,1", Rotation="0,0,0", Scale="1,1,1") {
    translationMode=BMM_Replace
    translationSpace=BCS_BoneSpace
    alphaBoolBlend {blendOption=Linear}
    boneToModify {boneName=dyn_can_mid_base}
    translation {z=1}
   }
   n14 = "Transform (Modify) Bone - Bone: dyn_can_front_base"(ComponentPose=@n13, Alpha=@n12, Translation="-0.000,-0.000,1", Rotation="0,0,0", Scale="1,1,1") {
    translationMode=BMM_Replace
    translationSpace=BCS_BoneSpace
    alphaBoolBlend {blendOption=Linear}
    boneToModify {boneName=dyn_can_front_base}
    translation {z=1}
   }
   n15 = RigidBody(ComponentPose=@n14) {
    bClampLinearTranslationLimitToRefPose=true
    bOverrideWorldGravity=true
    overridePhysicsAsset=&/F_MED_ChillCat_Necklace_Physics
    simulationSpace=BaseBoneSpace
    alphaBoolBlend {blendOption=Linear}
    baseBoneRef {boneName=spine_05}
    componentAppliedLinearAccClamp=<6000 6000 6000>
    componentLinearAccScale=<0.2 0.2 0.25>
    componentLinearVelScale=<0.2 0.2 0.25>
    overrideWorldGravity {z=-980}
    simSpaceSettings {
     maxAngularAcceleration=500
     maxAngularVelocity=3
     maxLinearAcceleration=400
     maxLinearVelocity=250
     velocityScaleZ=0
     worldAlpha=1
     externalLinearDragV=<0.1 0.1>
    }
   }
   n16 = RigidBody(ComponentPose=@n15) {
    bClampLinearTranslationLimitToRefPose=true
    bOverrideWorldGravity=true
    overridePhysicsAsset=&/F_MED_ChillCat_Physics
    simulationSpace=BaseBoneSpace
    alphaBoolBlend {blendOption=Linear}
    baseBoneRef {boneName=thigh_r}
    componentAppliedLinearAccClamp=<6000 6000 6000>
    componentLinearAccScale=<0.15 0.15 0.15>
    componentLinearVelScale=<0.15 0.15 0.15>
    overrideWorldGravity {z=-980}
    simSpaceSettings {
     maxAngularAcceleration=500
     maxAngularVelocity=3
     maxLinearAcceleration=400
     maxLinearVelocity=0
     velocityScaleZ=0
     worldAlpha=1
     externalLinearDragV=<0.1 0.1>
    }
   }
   Output Pose(Result=Component To Local(ComponentPose=@n16)) {layerGroup=None}
  ]}
 EventGraph {nodes [
   Event BlueprintUpdateAnimation()
   > n1 = Sequence!(execute) -> (then_0, then_1, then_2:Exec)
   > Set FrontEndCheck(FrontEndCheck:double=SelectFloat(A=1, B=0, bPickA=Get bIsInFrontEnd())) -> (Output_Get:double)
   Set LOD_Check(execute=@n1.then_1, LOD_Check:bool="integer <= integer"(A=Get CurrentLOD(), B=1)) -> (Output_Get:bool)
   n2 = Get bIsSkydiveFloating()
   n3 = Get bIsSkydiveDiveMode()
   Set Skydive(execute=@n1.then_2, Skydive:double=SelectFloat(A=1, B=0, bPickA=OR Boolean(A=@n2, B=@n3))) -> (Output_Get:double)
  ]}
]
variables [
 FrontEndCheck {type=number}
 LOD_Check {type=boolean}
 Skydive {type=number}
]
