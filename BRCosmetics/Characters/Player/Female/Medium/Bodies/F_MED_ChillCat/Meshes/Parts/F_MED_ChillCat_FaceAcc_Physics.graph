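# Physics asset for the face accessory mesh of the ChillCat female player cosmetic, covering the beanie and both cat ears.
# Small spheres on the base bones anchor capsule bodies: one for the beanie and a two-segment chain for each ear.
# Constraints with tight swing and near-locked twist limits allow only slight wobble.
# Left and right ears differ in capsule proportions and swing limits rather than mirroring exactly.
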PhysicsAsset /BRCosmetics/Characters/Player/Female/Medium/Bodies/F_MED_ChillCat/Meshes/Parts/F_MED_ChillCat_FaceAcc_Physics
bodies {
 dyn_beanie_base [{shapeType=Sphere, radius=0.5, shapeName=dyn_beanie_base_sphere}]
 dyn_beanie_parent [{shapeType=Capsule, length=3.1822, radius=2.1199, shapeName=dyn_beanie_parent_capsule, center=<2.591 0 0>, rotation=<rot 90 0 0>}]
 dyn_ear_01_l [{shapeType=Capsule, length=2.5253, radius=3.2047, shapeName=dyn_ear_parent_l_capsule, center=<2.8655 0 0>, rotation=<rot 90 0 0>}]
 dyn_ear_01_r [{shapeType=Capsule, length=4.4425, radius=2.3803, shapeName=dyn_ear_parent_l_capsule, center=<3.2136 0 0>, rotation=<rot 90 0 0>}]
 dyn_ear_base_l [{shapeType=Sphere, radius=0.5, shapeName=dyn_ear_base_l_sphere}]
 dyn_ear_base_r [{shapeType=Sphere, radius=0.5, shapeName=dyn_ear_base_r_sphere}]
 dyn_ear_parent_l [{shapeType=Capsule, length=2.241, radius=3.489, shapeName=dyn_ear_parent_l_capsule, center=<3.5735 0 0>, rotation=<rot 90 0 0>}]
 dyn_ear_parent_r [{shapeType=Capsule, length=5.5473, radius=2.429, shapeName=dyn_ear_parent_l_capsule, center=<4.9065 0 0>, rotation=<rot 90 0 0>}]
}
constraints [
 {bone1Name=dyn_ear_parent_r, bone2Name=dyn_ear_base_r, swing1LimitDegrees=6, swing2LimitDegrees=6, twistLimitDegrees=1}
 {bone1Name=dyn_ear_01_r, bone2Name=dyn_ear_parent_r, swing1LimitDegrees=10, swing2LimitDegrees=15, twistLimitDegrees=1}
 {bone1Name=dyn_ear_01_l, bone2Name=dyn_ear_parent_l, swing1LimitDegrees=10, swing2LimitDegrees=10, twistLimitDegrees=1}
 {bone1Name=dyn_ear_parent_l, bone2Name=dyn_ear_base_l, swing1LimitDegrees=8, swing2LimitDegrees=8, twistLimitDegrees=1}
 {bone1Name=dyn_beanie_parent, bone2Name=dyn_beanie_base, swing1LimitDegrees=10, swing2LimitDegrees=10, twistLimitDegrees=2}
]
