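# Animation Blueprint for the ChillCat female face accessory part, posing and simulating its ear bones and switching its mouth material.
# It copies the attached parent's pose and curves, applies additive ear-base rotations when the parachute is open, then runs RigidBody physics with damped component motion.
# Each update sets the ear alpha from the parachute state and switches the smile-mouth material frame depending on whether an emote plays.
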AnimBlueprint /BRCosmetics/Characters/Player/Female/Medium/Bodies/F_MED_ChillCat/Meshes/Parts/F_MED_ChillCat_FaceAcc_AnimBP
target_skeleton=&/F_MED_ChillCat_FaceAcc_Skeleton
graphs [
 AnimGraph {nodes [
   n1 = Get ParachuteOpen_Alpha() -> (ParachuteOpen_Alpha:double)
   n2 = "Transform (Modify) Bone - Bone: dyn_ear_base_r"(
    ComponentPose=Local To Component(LocalPose=Copy Pose From Mesh() {bCopyCurves=true, bUseAttachedParent=true})
    Alpha=@n1
    Translation="0,0,0"
    Rotation="-20,-60,0"
    Scale="1,1,1"
   ) {
    rotationMode=BMM_Additive
    rotationSpace=BCS_BoneSpace
    alphaBoolBlend {blendOption=Linear}
    boneToModify {boneName=dyn_ear_base_r}
    rotation {pitch=-20, yaw=-60}
   }
   n3 = "Transform (Modify) Bone - Bone: dyn_ear_base_l"(ComponentPose=@n2, Alpha=@n1, Translation="0,0,0", Rotation="-10,40,0", Scale="1,1,1") {
    rotationMode=BMM_Additive
    rotationSpace=BCS_BoneSpace
    alphaBoolBlend {blendOption=Linear}
    boneToModify {boneName=dyn_ear_base_l}
    rotation {pitch=-10, yaw=40}
   }
   n4 = RigidBody(ComponentPose=@n3) {
    bClampLinearTranslationLimitToRefPose=true
    bOverrideWorldGravity=true
    overridePhysicsAsset=&/F_MED_ChillCat_FaceAcc_Physics
    simulationSpace=BaseBoneSpace
    alphaBoolBlend {blendOption=Linear}
    baseBoneRef {boneName=spine_05}
    componentAppliedLinearAccClamp=<6000 6000 6000>
    componentLinearAccScale=<0.2 0.2 0.2>
    componentLinearVelScale=<0.2 0.2 0.2>
    overrideWorldGravity {z=-980}
    simSpaceSettings {
     maxAngularAcceleration=500
     maxAngularVelocity=3
     maxLinearAcceleration=400
     maxLinearVelocity=250
     velocityScaleZ=0
     worldAlpha=1
     externalLinearDragV=<0.1 0.1>
    }
   }
   Output Pose(Result=Component To Local(ComponentPose=@n4)) {layerGroup=None}
  ]}
 EventGraph {nodes [
   Event BlueprintUpdateAnimation()
   > n1 = Sequence()
   > n2 = Get Dynamic Material Instance(SkeletalMeshComponent=GetPartSkeletalMeshComponent())
   > n3 = Branch(Condition=Get bIsPlayingEmote())
   > SetScalarParameterValue(self=@n2, ParameterName=Smile Mouth animation Frame, Value=8)
   SetScalarParameterValue(execute=@n3.else, self=@n2, ParameterName=Smile Mouth animation Frame, Value=1)
   Set ParachuteOpen_Alpha(execute=@n1.then_1, ParachuteOpen_Alpha:double=SelectFloat(A=1, B=0, bPickA=Get bIsParachuteOpened())) -> (Output_Get:double)
   Event BlueprintInitializeAnimation()
   > Set skelmesh(skelmesh:Skeletal Mesh Component Object Reference=GetPartSkeletalMeshComponent()) -> (Output_Get:Skeletal Mesh Component Object Reference)
   > n4 = Set Owner_Skel_Mesh(Owner_Skel_Mesh:Skeletal Mesh Component Object Reference=GetOwnerSkeletalMeshComponent()) -> (Output_Get:Skeletal Mesh Component Object Reference)
   > Set Anim_Inst(Anim_Inst:Anim Instance Object Reference=GetAnimInstance(self=@n4)) -> (Output_Get:Anim Instance Object Reference)
  ]}
]
variables [
 Anim_Inst {type=AnimInstance, default=null}
 Owner_Skel_Mesh {type=SkeletalMeshComponent}
 ParachuteOpen_Alpha {type=number}
 skelmesh {type=SkeletalMeshComponent}
]
