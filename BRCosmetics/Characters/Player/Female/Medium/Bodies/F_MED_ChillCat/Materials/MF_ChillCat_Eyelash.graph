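# Material function for the ChillCat character's eyelashes that collapses masked vertices during a blink, effectively hiding them.
# It thresholds the stronger of the left and right blink pose parameters, so either eye triggers it, and multiplies that by a binary mask selecting vertices without full green vertex color.
# The result adds a World Position Offset pulling those vertices to the object position, behind the Use ChillCatEyelash static switch.
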MaterialFunction /BRCosmetics/Characters/Player/Female/Medium/Bodies/F_MED_ChillCat/Materials/MF_ChillCat_Eyelash
graph {nodes [
  n1 = GetMaterialAttributes(None=FunctionInput() {inputType=FunctionInput_MaterialAttributes}) {attributeGetTypes [#1]}
  n2 = Subtract(A=WorldPosition().XYZ, B=ObjectPositionWS())
  n3 = Max(A=ScalarParameter() {group=ChillCat, parameterName=R_blink_pose}, B=ScalarParameter() {group=ChillCat, parameterName=L_blink_pose})
  n4 = Ceil(None=Subtract(A=Saturate(None=@n3), B=ScalarParameter() {defaultValue=0.75, group=ChillCat, parameterName=BlinkPoseValue}))
  n5 = VertexColor()
  n6 = Ceil(None=Subtract(A=OneMinus(None=@n5.G)) {constB=0.01})
  n7 = Add(A=@n1.WorldPositionOffset, B=Multiply(A=Multiply(A=@n2) {constB=-1}, B=Multiply(A=@n4, B=@n6)))
  FunctionOutput(None=SetMaterialAttributes!(
    MaterialAttributes=@n1.MaterialAttributes
    WorldPositionOffset=StaticSwitchParameter(True=@n7, False=@n1.WorldPositionOffset) {parameterName=Use ChillCatEyelash}
   ) {attributeSetTypes [#1]})
 ]}
